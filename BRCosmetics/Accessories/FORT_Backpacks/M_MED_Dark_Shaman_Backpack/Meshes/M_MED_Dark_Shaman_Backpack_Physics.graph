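# Physics asset for the Dark Shaman backpack cosmetic, giving its two hanging ribbons collision bodies and joint limits for physical motion.
# The backpack root carries a sphere and a rotated box, and the left and right ribbons are each a chain of four capsule bodies.
# Each segment is constrained to the one before it, starting at the root, with moderate swing and very tight twist limits so the ribbons bend without spinning.
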PhysicsAsset /BRCosmetics/Accessories/FORT_Backpacks/M_MED_Dark_Shaman_Backpack/Meshes/M_MED_Dark_Shaman_Backpack_Physics
bodies {
 backpack_root [
  {shapeType=Sphere, radius=12.8215, center=<-4.0985 0.5763 -0.2379>}
  {shapeType=Box, extentX=32.1192, extentY=7.5, extentZ=33.1875, center=<-10.0408 1.9157 0.0007>, rotation=<rot 46.7622 14.4316 -0.2067>}
 ]
 dyn_l_ribbon_01 [{shapeType=Capsule, length=1.6962, radius=1.75, center=<5.7516 0.1523 0.1105>, rotation=<rot 2.2893 -89.9996 91.4635>}]
 dyn_l_ribbon_02 [{shapeType=Capsule, length=2.2273, radius=1.6313, center=<5.1139 0.19 0.003>, rotation=<rot 0 90.0037 -89.8999>}]
 dyn_l_ribbon_03 [{shapeType=Capsule, length=3.5776, radius=1.5243, center=<4.2926 -0.0007 -0.0257>, rotation=<rot 0 90.0404 -88.4561>}]
 dyn_l_ribbon_04 [{shapeType=Capsule, length=3.0477, radius=1.8362, center=<3.7274 -0.0041 0.1218>, rotation=<rot 0 90.484 -87.5206>}]
 dyn_r_ribbon_01 [{shapeType=Capsule, length=1.3582, radius=1.5826, center=<5.3602 0.0405 0.1024>, rotation=<rot 2.3706 -89.9994 89.7095>}]
 dyn_r_ribbon_02 [{shapeType=Capsule, length=2.3287, radius=1.6371, center=<5.3347 0.19 0.0047>, rotation=<rot 0 90.0056 -89.857>}]
 dyn_r_ribbon_03 [{shapeType=Capsule, length=3.1285, radius=1.7144, center=<4.4609 0.0006 -0.0055>, rotation=<rot 0 90.4513 -88.9321>}]
 dyn_r_ribbon_04 [{shapeType=Capsule, length=3.6389, radius=1.6518, center=<3.8195 -0.0008 0.1072>, rotation=<rot 0 89.8774 -89.8081>}]
}
constraints [
 {bone1Name=dyn_l_ribbon_01, bone2Name=backpack_root, swing1LimitDegrees=45, swing2LimitDegrees=45, twistLimitDegrees=2}
 {bone1Name=dyn_l_ribbon_02, bone2Name=dyn_l_ribbon_01, swing1LimitDegrees=45, swing2LimitDegrees=45, twistLimitDegrees=2}
 {bone1Name=dyn_l_ribbon_03, bone2Name=dyn_l_ribbon_02, swing1LimitDegrees=45, swing2LimitDegrees=45, twistLimitDegrees=2}
 {bone1Name=dyn_l_ribbon_04, bone2Name=dyn_l_ribbon_03, swing1LimitDegrees=45, swing2LimitDegrees=45, twistLimitDegrees=2}
 {bone1Name=dyn_r_ribbon_01, bone2Name=backpack_root, swing1LimitDegrees=45, swing2LimitDegrees=45, twistLimitDegrees=2}
 {bone1Name=dyn_r_ribbon_02, bone2Name=dyn_r_ribbon_01, swing1LimitDegrees=45, swing2LimitDegrees=45, twistLimitDegrees=2}
 {bone1Name=dyn_r_ribbon_03, bone2Name=dyn_r_ribbon_02, swing1LimitDegrees=45, swing2LimitDegrees=45, twistLimitDegrees=2}
 {bone1Name=dyn_r_ribbon_04, bone2Name=dyn_r_ribbon_03, swing1LimitDegrees=45, swing2LimitDegrees=45, twistLimitDegrees=2}
]
