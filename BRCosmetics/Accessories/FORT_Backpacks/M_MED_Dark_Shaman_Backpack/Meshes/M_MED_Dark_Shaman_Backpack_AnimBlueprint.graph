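# Animation Blueprint for the Dark Shaman backpack cosmetic, animating its ribbons over the incoming pose.
# Each update maps spine yaw to an additive backpack_root offset and derives a noisy vertical skydiving force while descending, reduced under parachute.
# bRigidBodyEnabled selects between RigidBody simulation receiving that force and plane-limited Trail controllers on both ribbons.
# The Anim Dynamics chain and a third Trail controller are not connected to the output.
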AnimBlueprint /BRCosmetics/Accessories/FORT_Backpacks/M_MED_Dark_Shaman_Backpack/Meshes/M_MED_Dark_Shaman_Backpack_AnimBlueprint
target_skeleton=&/M_MED_Dark_Shaman_Backpack_Skeleton
graphs [
 AnimGraph {nodes [
   n1 = Anim Dynamics - Bone: dyn_l_ribbon_01(Alpha=1, ExternalForce="0,0,0") {
    angularDampingOverride=0.8
    bGravityOverrideInSimSpace=true
    bOverrideAngularDamping=true
    bUseGravityOverride=true
    alphaBoolBlend {blendOption=Linear}
    boundBone {boneName=dyn_l_ribbon_01}
    componentLinearAccScale=<0.8 0.8 0.8>
    componentLinearVelScale=<0.8 0.8 0.8>
    gravityOverride {z=-980}
    physicsBodyDefinitions [{
      sphereCollisionRadius=0
      boundBone {boneName=dyn_l_ribbon_01}
      boxExtents=<9 3 3>
      constraintSetup {angularLimitsMax=<0 20 35>, angularLimitsMin=<0 -20 -2>, angularTarget=<0 0 0>}
      localJointOffset=<0.2 0 0>
     }]
   }
   n2 = Anim Dynamics - Bone: dyn_l_ribbon_02(ComponentPose=@n1, Alpha=1, ExternalForce="0,0,0") {
    angularDampingOverride=0.8
    bGravityOverrideInSimSpace=true
    bOverrideAngularDamping=true
    bUseGravityOverride=true
    alphaBoolBlend {blendOption=Linear}
    boundBone {boneName=dyn_l_ribbon_02}
    componentLinearAccScale=<0.8 0.8 0.8>
    componentLinearVelScale=<0.8 0.8 0.8>
    gravityOverride {z=-980}
    physicsBodyDefinitions [{
      sphereCollisionRadius=0
      boundBone {boneName=dyn_l_ribbon_02}
      boxExtents=<9 3 3>
      constraintSetup {angularLimitsMax=<0 20 35>, angularLimitsMin=<0 -20 -2>, angularTarget=<0 0 0>}
      localJointOffset=<0.2 0 0>
     }]
   }
   n3 = Anim Dynamics - Bone: dyn_l_ribbon_03(ComponentPose=@n2, Alpha=1, ExternalForce="0,0,0") {
    angularDampingOverride=0.8
    bGravityOverrideInSimSpace=true
    bOverrideAngularDamping=true
    bUseGravityOverride=true
    alphaBoolBlend {blendOption=Linear}
    boundBone {boneName=dyn_l_ribbon_03}
    componentLinearAccScale=<0.8 0.8 0.8>
    componentLinearVelScale=<0.8 0.8 0.8>
    gravityOverride {z=-980}
    physicsBodyDefinitions [{
      sphereCollisionRadius=0
      boundBone {boneName=dyn_l_ribbon_03}
      boxExtents=<9 3 3>
      constraintSetup {angularLimitsMax=<0 20 35>, angularLimitsMin=<0 -20 -2>, angularTarget=<0 0 0>}
      localJointOffset=<0.2 0 0>
     }]
   }
   n4 = Anim Dynamics - Bone: dyn_l_ribbon_04(ComponentPose=@n3, Alpha=1, ExternalForce="0,0,0") {
    angularDampingOverride=0.8
    bGravityOverrideInSimSpace=true
    bOverrideAngularDamping=true
    bUseGravityOverride=true
    alphaBoolBlend {blendOption=Linear}
    boundBone {boneName=dyn_l_ribbon_04}
    componentLinearAccScale=<0.8 0.8 0.8>
    componentLinearVelScale=<0.8 0.8 0.8>
    gravityOverride {z=-980}
    physicsBodyDefinitions [{
      sphereCollisionRadius=0
      boundBone {boneName=dyn_l_ribbon_04}
      boxExtents=<9 3 3>
      constraintSetup {angularLimitsMax=<0 20 35>, angularLimitsMin=<0 -20 -2>, angularTarget=<0 0 0>}
      localJointOffset=<0.2 0 0>
     }]
   }
   n5 = Anim Dynamics - Bone: dyn_r_ribbon_01(ComponentPose=@n4, Alpha=1, ExternalForce="0,0,0") {
    angularDampingOverride=0.8
    bGravityOverrideInSimSpace=true
    bOverrideAngularDamping=true
    bUseGravityOverride=true
    alphaBoolBlend {blendOption=Linear}
    boundBone {boneName=dyn_r_ribbon_01}
    componentLinearAccScale=<0.7 0.7 0.7>
    componentLinearVelScale=<0.7 0.7 0.7>
    gravityOverride {z=-980}
    physicsBodyDefinitions [{
      sphereCollisionRadius=0
      boundBone {boneName=dyn_r_ribbon_01}
      boxExtents=<7 5 5>
      constraintSetup {angularLimitsMax=<0 20 35>, angularLimitsMin=<0 -20 -2>, angularTarget=<0 0 0>}
      localJointOffset=<0.2 0 0>
     }]
   }
   n6 = Anim Dynamics - Bone: dyn_r_ribbon_02(ComponentPose=@n5, Alpha=1, ExternalForce="0,0,0") {
    angularDampingOverride=0.8
    bGravityOverrideInSimSpace=true
    bOverrideAngularDamping=true
    bUseGravityOverride=true
    alphaBoolBlend {blendOption=Linear}
    boundBone {boneName=dyn_r_ribbon_02}
    componentLinearAccScale=<0.7 0.7 0.7>
    componentLinearVelScale=<0.7 0.7 0.7>
    gravityOverride {z=-980}
    physicsBodyDefinitions [{
      sphereCollisionRadius=0
      boundBone {boneName=dyn_r_ribbon_02}
      boxExtents=<7 5 5>
      constraintSetup {angularLimitsMax=<0 20 35>, angularLimitsMin=<0 -20 -2>, angularTarget=<0 0 0>}
      localJointOffset=<0.2 0 0>
     }]
   }
   n7 = Anim Dynamics - Bone: dyn_r_ribbon_03(ComponentPose=@n6, Alpha=1, ExternalForce="0,0,0") {
    angularDampingOverride=0.8
    bGravityOverrideInSimSpace=true
    bOverrideAngularDamping=true
    bUseGravityOverride=true
    alphaBoolBlend {blendOption=Linear}
    boundBone {boneName=dyn_r_ribbon_03}
    componentLinearAccScale=<0.7 0.7 0.7>
    componentLinearVelScale=<0.7 0.7 0.7>
    gravityOverride {z=-980}
    physicsBodyDefinitions [{
      sphereCollisionRadius=0
      boundBone {boneName=dyn_r_ribbon_03}
      boxExtents=<7 5 5>
      constraintSetup {angularLimitsMax=<0 20 35>, angularLimitsMin=<0 -20 -2>, angularTarget=<0 0 0>}
      localJointOffset=<0.2 0 0>
     }]
   }
   Anim Dynamics - Bone: dyn_r_ribbon_04(ComponentPose=@n7, Alpha=1, ExternalForce="0,0,0") {
    angularDampingOverride=0.8
    bGravityOverrideInSimSpace=true
    bOverrideAngularDamping=true
    bUseGravityOverride=true
    alphaBoolBlend {blendOption=Linear}
    boundBone {boneName=dyn_r_ribbon_04}
    componentLinearAccScale=<0.7 0.7 0.7>
    componentLinearVelScale=<0.7 0.7 0.7>
    gravityOverride {z=-980}
    physicsBodyDefinitions [{
      sphereCollisionRadius=0
      boundBone {boneName=dyn_r_ribbon_04}
      boxExtents=<7 5 5>
      constraintSetup {angularLimitsMax=<0 20 35>, angularLimitsMin=<0 -20 -2>, angularTarget=<0 0 0>}
      localJointOffset=<0.2 0 0>
     }]
   }
   Local Space Ref Pose()
   n8 = "Transform (Modify) Bone - Bone: backpack_root"(
    ComponentPose=Local To Component(LocalPose=Use cached pose 'In Pose'())
    Alpha=1
    Translation=Get backpack_translation()
    Rotation=Get backpack_rotate()
    Scale="1,1,1"
   ) {
    rotationMode=BMM_Additive
    rotationSpace=BCS_BoneSpace
    translationMode=BMM_Additive
    translationSpace=BCS_BoneSpace
    alphaBoolBlend {blendOption=Linear}
    boneToModify {boneName=backpack_root}
   }
   n9 = RigidBody(ComponentPose=@n8, ExternalForce=Get skydiving_force() -> (skydiving_force:Vector)) {
    bOverrideWorldGravity=true
    overridePhysicsAsset=&/M_MED_Dark_Shaman_Backpack_Physics
    simulationSpace=BaseBoneSpace
    alphaBoolBlend {blendOption=Linear}
    baseBoneRef {boneName=backpack_root}
    componentAppliedLinearAccClamp=<3000 3000 3000>
    componentLinearAccScale=<0.2 0.2 0.2>
    componentLinearVelScale=<0.3 0.3 0.3>
    overrideWorldGravity {z=-580}
   }
   n10 = "Transform (Modify) Bone - Bone: backpack_root"(
    ComponentPose=Local To Component(LocalPose=Use cached pose 'In Pose'())
    Alpha=1
    Translation=Get backpack_translation()
    Rotation=Get backpack_rotate()
    Scale="1,1,1"
   ) {
    rotationMode=BMM_Additive
    rotationSpace=BCS_BoneSpace
    translationMode=BMM_Additive
    translationSpace=BCS_BoneSpace
    alphaBoolBlend {blendOption=Linear}
    boneToModify {boneName=backpack_root}
   }
   n11 = Trail controller - Bone: dyn_l_ribbon_04(ComponentPose=@n10, Alpha=1) {
    bUsePlanarLimit=true
    chainLength=4
    planarLimits [
     {drivingBone {boneName=backpack_root}, planeTransform {location=<-2 0 -15>, rotation=<rot -15 5.0001 -180>, scale=<0.5 0.5 0.5>}}
     {drivingBone {boneName=backpack_root}, planeTransform {location=<-13 -5 0>, rotation=<rot 44.9999 13.0007 -90>, scale=<0.3 0.3 0.3>}}
    ]
    rotationLimits [
     {limitMax=<180 180 180>}
     {limitMax=<0 10 180>, limitMin=<0 -45 -180>}
     {limitMax=<0 30 180>, limitMin=<0 -45 -180>}
     {limitMax=<0 30 180>, limitMin=<0 -45 -180>}
    ]
    rotationOffsets [<0 0 0>, <0 0 0>, <0 0 0>, <0 0 0>]
    trailBone {boneName=dyn_l_ribbon_04}
    trailRelaxationSpeed {editorCurveData {keys [
       {arriveTangent=-1.3013, interpMode=RCIM_Cubic, leaveTangent=-1.3013, tangentMode=RCTM_User, time=0, value=16}
       {arriveTangent=-16.4537, interpMode=RCIM_Cubic, leaveTangent=-16.4537, tangentMode=RCTM_User, time=0.25, value=12}
       {arriveTangent=-10.6667, interpMode=RCIM_Cubic, leaveTangent=-10.6667, time=0.4976, value=8}
       {arriveTangent=-3.0387, interpMode=RCIM_Cubic, leaveTangent=-3.0387, tangentMode=RCTM_User, time=1, value=4}
      ]}}
   }
   n12 = Trail controller - Bone: dyn_r_ribbon_04(ComponentPose=@n11, Alpha=1) {
    bUsePlanarLimit=true
    chainLength=4
    relaxationSpeedScale=2
    planarLimits [
     {drivingBone {boneName=backpack_root}, planeTransform {location=<-2 0 15>, rotation=<rot 15 5.0001 0>, scale=<0.5 0.5 0.5>}}
     {drivingBone {boneName=backpack_root}, planeTransform {location=<-13 -5 0>, rotation=<rot 44.9999 13.0008 -90>, scale=<0.3 0.3 0.3>}}
    ]
    rotationLimits [
     {limitMax=<180 180 180>}
     {limitMax=<0 10 180>, limitMin=<0 -45 -180>}
     {limitMax=<0 30 180>, limitMin=<0 -45 -180>}
     {limitMax=<0 30 180>, limitMin=<0 -45 -180>}
    ]
    rotationOffsets [<0 0 0>, <0 0 0>, <0 0 0>, <0 0 0>]
    trailBone {boneName=dyn_r_ribbon_04}
    trailRelaxationSpeed {editorCurveData {keys [
       {arriveTangent=-1.3013, interpMode=RCIM_Cubic, leaveTangent=-1.3013, tangentMode=RCTM_User, time=0, value=16}
       {arriveTangent=-16.4537, interpMode=RCIM_Cubic, leaveTangent=-16.4537, tangentMode=RCTM_User, time=0.25, value=12}
       {arriveTangent=-10.6667, interpMode=RCIM_Cubic, leaveTangent=-10.6667, time=0.4976, value=8}
       {arriveTangent=-3.0386, interpMode=RCIM_Cubic, leaveTangent=-3.0387, tangentMode=RCTM_User, time=1, value=4}
      ]}}
   }
   Output Pose(Result=Blend Poses by bool(BlendPose_0=Component To Local(ComponentPose=@n9), BlendPose_1=Component To Local(ComponentPose=@n12), bActiveValue=Get bRigidBodyEnabled()) {blendType=Linear}) {layerGroup=None}
   AnimGraphNode_SubInput_0 = Input Pose()
   Save cached pose 'In Pose'(Pose=@AnimGraphNode_SubInput_0) {cacheName=In Pose}
   Trail controller - Bone: dyn_l_ribbon_04(Alpha=1) {
    bUsePlanarLimit=true
    chainLength=4
    planarLimits [{drivingBone {boneName=backpack_root}, planeTransform {location=<0 0 0>, rotation=<rot 0 18.0001 -90>, scale=<2 2 2>}}]
    rotationLimits [{limitMax=<180 180 180>}, {limitMax=<180 180 180>}, {limitMax=<180 180 180>}, {limitMax=<180 180 180>}]
    rotationOffsets [<0 0 0>, <0 0 0>, <0 0 0>, <0 0 0>]
    trailBone {boneName=dyn_l_ribbon_04}
    trailRelaxationSpeed {editorCurveData {keys [{arriveTangent=-11.8713, interpMode=RCIM_Cubic, leaveTangent=-11.8713, tangentMode=RCTM_User, time=0, value=10}, <at 1 5 RCIM_Cubic>]}}
   }
  ]}
 EventGraph {nodes [
   Event BlueprintUpdateAnimation()
   n1 = GetOwnerSkeletalMeshComponent()
   n2 = BreakRotator(InRot=BreakTransform(InTransform=GetDeltaTransformFromRefPose(self=@n1, BoneName=spine_05, BaseName=spine_01)).Rotation)
   > Set backpack_rotate(backpack_rotate=MakeRotator(Yaw=MapRangeClamped(Value=@n2.Yaw, InRangeA=0, InRangeB=-30, OutRangeB=20)))
   > Set backpack_translation(backpack_translation=MakeVector(Y=MapRangeClamped(Value=@n2.Yaw, InRangeA=0, InRangeB=-30, OutRangeB=-3)))
   n3 = Get PawnVelocityZ()
   n4 = SelectFloat(
    A=Make Perlin Noise and Remap!(Value:float=@n3, RangeOutMin:float=100, RangeOutMax:float=500) -> (ReturnValue:float)
    B=0
    bPickA=AND Boolean(A=Get bIsSkyDiving(), B="float <= float"!(A:double=@n3, B:double=0) -> (ReturnValue:bool) {is=title K2Node_CallFunction})
   )
   n5 = float * float(A=@n4, B=0.1)
   n6 = Get bIsParachuteOpened()
   > Set skydiving_force(skydiving_force:Vector=MakeVector(Y=0, Z=SelectFloat(A=@n5, B=@n4, bPickA=@n6))) -> (Output_Get:Vector)
  ]}
]
variables [
 backpack_rotate {type=Rotator}
 backpack_translation {type=Vector}
 skydiving_force {type=Vector}
]
